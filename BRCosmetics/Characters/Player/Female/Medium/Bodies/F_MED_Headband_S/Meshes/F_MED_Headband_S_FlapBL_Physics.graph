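# Physics asset for the FlapBL skirt flap on the female medium Headband_S body cosmetic, apparently the back-left flap.
# Two bone chains (a and b) of box bodies each run from a master through three segments, with narrow swing and twist limits, tightest at the root of chain a.
# Angularly free constraints join both masters to the left thigh and link the two chain tips.
# Capsules represent both thighs.
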PhysicsAsset /BRCosmetics/Characters/Player/Female/Medium/Bodies/F_MED_Headband_S/Meshes/F_MED_Headband_S_FlapBL_Physics
bodies {
 dyn_skirt_bl_a_01 [{shapeType=Box, extentX=6.5, extentY=7, extentZ=8, shapeName=dyn_skirt_bl_a_01_box, center=<-2.1557 0 0>}]
 dyn_skirt_bl_a_02 [{shapeType=Box, extentX=7.5, extentY=7.5, extentZ=9, shapeName=dyn_skirt_bl_a_02_box, center=<-2.1557 0 0>}]
 dyn_skirt_bl_a_03 [{shapeType=Box, extentX=6.5, extentY=6, extentZ=8, shapeName=dyn_skirt_bl_a_03_box, center=<-2.1557 0 0>}]
 dyn_skirt_bl_a_master [{shapeType=Box, extentX=1.5, extentY=2, extentZ=2, shapeName=dyn_skirt_bl_a_master_box}]
 dyn_skirt_bl_b_01 [{shapeType=Box, extentX=6.5, extentY=4, extentZ=8, shapeName=dyn_skirt_bl_b_01_box, center=<-2.1557 -0.6985 0>}]
 dyn_skirt_bl_b_02 [{shapeType=Box, extentX=8.5, extentY=7, extentZ=10, shapeName=dyn_skirt_bl_b_02_box, center=<-2.1557 -0.6985 0>}]
 dyn_skirt_bl_b_03 [{shapeType=Box, extentX=6.5, extentY=6, extentZ=8, shapeName=dyn_skirt_bl_b_03_box, center=<-2.1557 -0.6985 0>}]
 dyn_skirt_bl_b_master [{shapeType=Box, extentX=1.5, extentY=2, extentZ=2, shapeName=dyn_skirt_bl_b_master_box}]
 thigh_l [{shapeType=Capsule, length=19.3589, radius=7.8941, shapeName=thigh_l_capsule, center=<-23.9699 0 0>, rotation=<rot 90 180 0>}]
 thigh_r [{shapeType=Capsule, length=19.3589, radius=7.8941, shapeName=thigh_r_capsule, center=<23.9699 0 0>, rotation=<rot 90 0 0>}]
}
constraints [
 {bone1Name=dyn_skirt_bl_b_03, bone2Name=dyn_skirt_bl_b_02, swing1LimitDegrees=15, swing2LimitDegrees=15, twistLimitDegrees=5}
 {bone1Name=dyn_skirt_bl_a_01, bone2Name=dyn_skirt_bl_a_master, swing1LimitDegrees=5, swing2LimitDegrees=5, twistLimitDegrees=2}
 {bone1Name=dyn_skirt_bl_a_02, bone2Name=dyn_skirt_bl_a_01, swing1LimitDegrees=15, swing2LimitDegrees=15, twistLimitDegrees=5}
 {bone1Name=dyn_skirt_bl_a_03, bone2Name=dyn_skirt_bl_a_02, swing1LimitDegrees=15, swing2LimitDegrees=15, twistLimitDegrees=5}
 {bone1Name=dyn_skirt_bl_b_01, bone2Name=dyn_skirt_bl_b_master, swing1LimitDegrees=15, swing2LimitDegrees=15, twistLimitDegrees=5}
 {bone1Name=dyn_skirt_bl_b_02, bone2Name=dyn_skirt_bl_b_01, swing1LimitDegrees=15, swing2LimitDegrees=15, twistLimitDegrees=5}
 {
  bone1Name=dyn_skirt_bl_b_03
  bone2Name=dyn_skirt_bl_a_03
  swing1LimitDegrees=45
  swing1Motion=Free
  swing2LimitDegrees=45
  swing2Motion=Free
  twistLimitDegrees=45
  twistMotion=Free
 }
 {
  bone1Name=dyn_skirt_bl_a_master
  bone2Name=thigh_l
  swing1LimitDegrees=45
  swing1Motion=Free
  swing2LimitDegrees=45
  swing2Motion=Free
  twistLimitDegrees=45
  twistMotion=Free
 }
 {
  bone1Name=dyn_skirt_bl_b_master
  bone2Name=thigh_l
  swing1LimitDegrees=45
  swing1Motion=Free
  swing2LimitDegrees=45
  swing2Motion=Free
  twistLimitDegrees=45
  twistMotion=Free
 }
]
