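# Physics asset for the FL skirt flap on the female medium Headband_S cosmetic body, likely front-left given its left-thigh anchor.
# Two three-bone chains (a, b) of box bodies hang from tiny master bodies, each joined to the thigh_l capsule with free rotation.
# Tight swing and twist limits, narrower at the roots, restrain each chain.
# A free-rotation constraint cross-links the two tip bodies, coupling the chains into one flap.
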PhysicsAsset /BRCosmetics/Characters/Player/Female/Medium/Bodies/F_MED_Headband_S/Meshes/F_MED_Headband_S_FlapFL_Physics
bodies {
 dyn_skirt_fl_a_01 [{shapeType=Box, extentX=7.5, extentY=1.5, extentZ=8.5, shapeName=dyn_skirt_fl_a_01_box, center=<0.1697 -0.0001 0>}]
 dyn_skirt_fl_a_02 [{shapeType=Box, extentX=7.5, extentY=1.5, extentZ=8.5, shapeName=dyn_skirt_fl_a_02_box, center=<0.1697 -0.0001 0>}]
 dyn_skirt_fl_a_03 [{shapeType=Box, extentX=7.5, extentY=1.5, extentZ=8.5, shapeName=dyn_skirt_fl_a_03_box, center=<0.1697 -0.0001 0>}]
 dyn_skirt_fl_a_master [{shapeType=Box, extentX=0.1, extentY=0.1, extentZ=0.1, shapeName=dyn_skirt_fl_a_master_box}]
 dyn_skirt_fl_b_01 [{shapeType=Box, extentX=7.5, extentY=1.5, extentZ=8.5, shapeName=dyn_skirt_fl_b_01_box, center=<0.1697 -0.0001 0>}]
 dyn_skirt_fl_b_02 [{shapeType=Box, extentX=7.5, extentY=2.5, extentZ=8.5, shapeName=dyn_skirt_fl_b_02_box, center=<0.1697 -0.0001 0>}]
 dyn_skirt_fl_b_03 [{shapeType=Box, extentX=7.5, extentY=2.5, extentZ=8.5, shapeName=dyn_skirt_fl_b_03_box, center=<0.1697 -0.0001 0>}]
 dyn_skirt_fl_b_master [{shapeType=Box, extentX=0.1, extentY=0.1, extentZ=0.1, shapeName=dyn_skirt_fl_b_master_box}]
 thigh_l [{shapeType=Capsule, length=7.75, radius=6.75, shapeName=thigh_l_capsule, center=<-21.4862 0.7517 -0.0001>, rotation=<rot 90 -5.7992 -2.9233>}]
}
constraints [
 {
  bone1Name=dyn_skirt_fl_a_master
  bone2Name=thigh_l
  swing1LimitDegrees=45
  swing1Motion=Free
  swing2LimitDegrees=45
  swing2Motion=Free
  twistLimitDegrees=45
  twistMotion=Free
 }
 {bone1Name=dyn_skirt_fl_a_01, bone2Name=dyn_skirt_fl_a_master, swing1LimitDegrees=8, swing2LimitDegrees=8, twistLimitDegrees=5}
 {bone1Name=dyn_skirt_fl_a_02, bone2Name=dyn_skirt_fl_a_01, swing1LimitDegrees=15, swing2LimitDegrees=15, twistLimitDegrees=5}
 {bone1Name=dyn_skirt_fl_a_03, bone2Name=dyn_skirt_fl_a_02, swing1LimitDegrees=15, swing2LimitDegrees=15, twistLimitDegrees=5}
 {
  bone1Name=dyn_skirt_fl_b_master
  bone2Name=thigh_l
  swing1LimitDegrees=45
  swing1Motion=Free
  swing2LimitDegrees=45
  swing2Motion=Free
  twistLimitDegrees=45
  twistMotion=Free
 }
 {bone1Name=dyn_skirt_fl_b_01, bone2Name=dyn_skirt_fl_b_master, swing1LimitDegrees=8, swing2LimitDegrees=8, twistLimitDegrees=5}
 {bone1Name=dyn_skirt_fl_b_02, bone2Name=dyn_skirt_fl_b_01, swing1LimitDegrees=15, swing2LimitDegrees=15, twistLimitDegrees=5}
 {bone1Name=dyn_skirt_fl_b_03, bone2Name=dyn_skirt_fl_b_02, swing1LimitDegrees=15, swing2LimitDegrees=15, twistLimitDegrees=5}
 {
  bone1Name=dyn_skirt_fl_a_03
  bone2Name=dyn_skirt_fl_b_03
  swing1LimitDegrees=45
  swing1Motion=Free
  swing2LimitDegrees=45
  swing2Motion=Free
  twistLimitDegrees=45
  twistMotion=Free
 }
]
