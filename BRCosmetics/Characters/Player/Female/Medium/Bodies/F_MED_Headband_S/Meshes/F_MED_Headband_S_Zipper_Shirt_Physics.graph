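# Physics asset for the zipper on the shirt of the F_MED_Headband_S female player cosmetic body.
# It has capsule bodies on the spine_05 bone and two dynamic zipper bones, dyn_zipper2_01 and dyn_zipper2_02.
# Constraints chain the second zipper bone to the first and the first to spine_05, with tight swing and tighter twist limits that allow only slight dangling motion.
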PhysicsAsset /BRCosmetics/Characters/Player/Female/Medium/Bodies/F_MED_Headband_S/Meshes/F_MED_Headband_S_Zipper_Shirt_Physics
bodies {
 dyn_zipper2_01 [{shapeType=Capsule, length=0.505, radius=1, shapeName=dyn_zipper2_01_capsule, center=<1.4371 0.0072 -0.0032>, rotation=<rot 90 0 0>}]
 dyn_zipper2_02 [{shapeType=Capsule, length=0.505, radius=0.505, shapeName=dyn_zipper2_02_capsule, center=<0.4231 0.0001 0>, rotation=<rot 0 -90 0>}]
 spine_05 [{shapeType=Capsule, length=1, radius=1, shapeName=spine_05_capsule}]
}
constraints [
 {bone1Name=dyn_zipper2_02, bone2Name=dyn_zipper2_01, swing1LimitDegrees=15, swing2LimitDegrees=15, twistLimitDegrees=5}
 {bone1Name=dyn_zipper2_01, bone2Name=spine_05, swing1LimitDegrees=15, swing2LimitDegrees=15, twistLimitDegrees=5}
]
