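# Physics asset for the zipper on the skirt of the female medium Headband_S cosmetic body, covering its dyn_zipper1 bone chain.
# Three small capsule bodies sit on the base, 01 and 02 bones, with the tip capsule the largest.
# Two constraints link them in series with matching narrow swing and twist limits, restricting how far the zipper can bend and twist.
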PhysicsAsset /BRCosmetics/Characters/Player/Female/Medium/Bodies/F_MED_Headband_S/Meshes/F_MED_Headband_S_Zipper_Skirt_Physics
bodies {
 dyn_zipper1_01 [{shapeType=Capsule, length=0.505, radius=0.505, shapeName=dyn_zipper1_01_capsule, center=<0.3353 0 0>, rotation=<rot 90 0 0>}]
 dyn_zipper1_02 [{shapeType=Capsule, length=1.005, radius=0.755, shapeName=dyn_zipper1_02_capsule, center=<0.9078 0 0>, rotation=<rot 90 0 0>}]
 dyn_zipper1_base [{shapeType=Capsule, length=0.75, radius=0.5531, shapeName=dyn_zipper1_base_capsule}]
}
constraints [
 {bone1Name=dyn_zipper1_02, bone2Name=dyn_zipper1_01, swing1LimitDegrees=25, swing2LimitDegrees=15, twistLimitDegrees=5}
 {bone1Name=dyn_zipper1_01, bone2Name=dyn_zipper1_base, swing1LimitDegrees=25, swing2LimitDegrees=15, twistLimitDegrees=5}
]
