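# Post-process-style AnimBP for the F_MED_Headband_S female body, adding secondary motion to skirt flaps and zippers over the parent mesh's copied pose.
# Separate front-end and gameplay chains, selected by bIsInFrontEnd, layer additive bone corrections around RigidBody simulations.
# Update event derives thigh rotations and corrective alphas for crouch, emote and front-end poses from thigh and pelvis deltas.
# It also sets FPS-based flap alphas, which the AnimGraph never reads.
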AnimBlueprint /BRCosmetics/Characters/Player/Female/Medium/Bodies/F_MED_Headband_S/Meshes/F_MED_Headband_S_AnimBP
target_skeleton=&/F_MED_Headband_S_Skeleton
graphs [
 AnimGraph {nodes [
   n1 = Get Pelvis FrontEnd Pose Alpha() -> (Pelvis FrontEnd Pose Alpha:double)
   n2 = "Transform (Modify) Bone - Bone: dyn_skirt_bl_b_master"(
    ComponentPose=Local To Component(LocalPose=Copy Pose From Mesh() {bUseAttachedParent=true})
    Alpha=@n1
    Translation="0.000,-2,0.000"
    Rotation="0,0,0"
    Scale="1,1,1"
   ) {
    rotationMode=BMM_Additive
    rotationSpace=BCS_BoneSpace
    translationMode=BMM_Additive
    translationSpace=BCS_BoneSpace
    alphaBoolBlend {blendOption=Linear}
    alphaScaleBiasClamp {bInterpResult=true}
    boneToModify {boneName=dyn_skirt_bl_b_master}
    translation {y=-2}
   }
   n3 = "Transform (Modify) Bone - Bone: dyn_skirt_bl_b_02"(ComponentPose=@n2, Alpha=@n1, Translation="0,0,0", Rotation="-10,20.000,0.000", Scale="1,1,1") {
    rotationMode=BMM_Additive
    rotationSpace=BCS_BoneSpace
    translationMode=BMM_Additive
    translationSpace=BCS_BoneSpace
    alphaBoolBlend {blendOption=Linear}
    alphaScaleBiasClamp {bInterpResult=true}
    boneToModify {boneName=dyn_skirt_bl_b_02}
    rotation {pitch=-10, yaw=20}
   }
   n4 = "Transform (Modify) Bone - Bone: dyn_skirt_l_master"(ComponentPose=@n3, Alpha=1, Translation="0,0,0", Rotation=Get Thigh Left Rot() -> (Thigh Left Rot:Rotator), Scale="1,1,1") {
    rotationMode=BMM_Additive
    rotationSpace=BCS_BoneSpace
    translationMode=BMM_Additive
    alphaBoolBlend {blendOption=Linear}
    alphaScaleBiasClamp {bInterpResult=true}
    boneToModify {boneName=dyn_skirt_l_master}
   }
   n5 = "Transform (Modify) Bone - Bone: dyn_skirt_fr_master"(ComponentPose=@n4, Alpha=1, Translation="0,0,0", Rotation=Get Thigh Right Rot() -> (Thigh Right Rot:Rotator), Scale="1,1,1") {
    rotationMode=BMM_Additive
    rotationSpace=BCS_BoneSpace
    translationMode=BMM_Additive
    alphaBoolBlend {blendOption=Linear}
    alphaScaleBiasClamp {bInterpResult=true}
    boneToModify {boneName=dyn_skirt_fr_master}
   }
   n6 = RigidBody(ComponentPose=@n5, Alpha=0.7) {
    bOverrideWorldGravity=true
    overridePhysicsAsset=&/F_MED_Headband_S_Flap_Physics
    simulationSpace=BaseBoneSpace
    alphaBoolBlend {blendOption=Linear}
    baseBoneRef {boneName=pelvis}
    componentAppliedLinearAccClamp=<2000 2000 2000>
    componentLinearAccScale=<0.3 0.3 0.35>
    componentLinearVelScale=<0.3 0.3 0.35>
    overrideWorldGravity {z=-980}
   }
   n7 = RigidBody(ComponentPose=@n6, Alpha=0.3) {
    bOverrideWorldGravity=true
    overridePhysicsAsset=&/F_MED_Headband_S_FlapFL_Physics
    simulationSpace=BaseBoneSpace
    alphaBoolBlend {blendInTime=2, blendOption=Linear, blendOutTime=1}
    alphaScaleBiasClamp {bInterpResult=true, interpSpeedIncreasing=2}
    baseBoneRef {boneName=pelvis}
    componentAppliedLinearAccClamp=<2000 2000 2000>
    componentLinearAccScale=<0.3 0.3 0.35>
    componentLinearVelScale=<0.3 0.3 0.35>
    overrideWorldGravity {z=-980}
   }
   n8 = RigidBody(ComponentPose=@n7) {
    bOverrideWorldGravity=true
    overridePhysicsAsset=&/F_MED_Headband_S_Zipper_Skirt_Physics
    simulationSpace=BaseBoneSpace
    alphaBoolBlend {blendInTime=2, blendOption=Linear, blendOutTime=1}
    alphaScaleBiasClamp {interpSpeedIncreasing=2}
    baseBoneRef {boneName=pelvis}
    componentAppliedLinearAccClamp=<2000 2000 2000>
    componentLinearAccScale=<0.3 0.3 0.35>
    componentLinearVelScale=<0.3 0.3 0.35>
    overrideWorldGravity {z=-980}
   }
   n9 = RigidBody(ComponentPose=@n8) {
    bOverrideWorldGravity=true
    overridePhysicsAsset=&/F_MED_Headband_S_Zipper_Shirt_Physics
    simulationSpace=BaseBoneSpace
    alphaBoolBlend {blendInTime=2, blendOption=Linear, blendOutTime=1}
    alphaScaleBiasClamp {interpSpeedIncreasing=2}
    baseBoneRef {boneName=spine_05}
    componentAppliedLinearAccClamp=<2000 2000 2000>
    componentLinearAccScale=<0.3 0.3 0.35>
    componentLinearVelScale=<0.3 0.3 0.35>
    overrideWorldGravity {z=-980}
   }
   n10 = "Transform (Modify) Bone - Bone: dyn_skirt_l_master"(ComponentPose=@n9, Alpha=1, Translation="0,0,0", Rotation="0,-2,0", Scale="1,1,1") {
    rotationMode=BMM_Additive
    rotationSpace=BCS_BoneSpace
    alphaBoolBlend {blendOption=Linear}
    boneToModify {boneName=dyn_skirt_l_master}
    rotation {yaw=-2}
   }
   n11 = "Transform (Modify) Bone - Bone: dyn_skirt_bl_a_master"(ComponentPose=@n10, Alpha=Get Headband Emote Alpha() -> (Headband Emote Alpha:double), Translation="0,5,0", Rotation="0,0,0", Scale="1,1,1") {
    rotationMode=BMM_Additive
    rotationSpace=BCS_BoneSpace
    translationMode=BMM_Additive
    translationSpace=BCS_BoneSpace
    alphaBoolBlend {blendInTime=1, blendOption=Linear, blendOutTime=1}
    alphaScaleBiasClamp {bInterpResult=true}
    boneToModify {boneName=dyn_skirt_bl_a_master}
    translation {y=5}
   }
   n12 = Get Pelvis FrontEnd Pose Inverse Alpha() -> (Pelvis FrontEnd Pose Inverse Alpha:double)
   n13 = "Transform (Modify) Bone - Bone: dyn_skirt_bl_a_master"(ComponentPose=@n11, Alpha=@n12, Translation="0,6,0", Rotation="-4.741,0.000,0.000", Scale="1,1,1") {
    rotationMode=BMM_Additive
    rotationSpace=BCS_BoneSpace
    translationMode=BMM_Additive
    translationSpace=BCS_BoneSpace
    alphaBoolBlend {blendInTime=1, blendOption=Linear, blendOutTime=1}
    boneToModify {boneName=dyn_skirt_bl_a_master}
    rotation {pitch=-4.741}
    translation {y=6}
   }
   n14 = "Transform (Modify) Bone - Bone: dyn_skirt_bl_a_01"(ComponentPose=@n13, Alpha=@n12, Translation="0.442,-0.000,3.320", Rotation="2.493,0.000,0.000", Scale="1,1,1") {
    rotationMode=BMM_Additive
    rotationSpace=BCS_BoneSpace
    translationMode=BMM_Additive
    translationSpace=BCS_BoneSpace
    alphaBoolBlend {blendInTime=1, blendOption=Linear, blendOutTime=1}
    boneToModify {boneName=dyn_skirt_bl_a_01}
    rotation {pitch=2.493}
    translation {x=0.442, z=3.32}
   }
   n15 = "Transform (Modify) Bone - Bone: dyn_skirt_bl_b_02"!(
    ComponentPose=@n14
    bAlphaBoolEnabled:bool=Get RBAN Disabled() -> (RBAN Disabled:bool)
    Translation="0,0,0"
    Rotation="10,-30,0"
    Scale="1,1,1"
   ) -> (Pose) {
    alphaInputType=Bool
    rotationMode=BMM_Additive
    rotationSpace=BCS_BoneSpace
    translationMode=BMM_Additive
    translationSpace=BCS_BoneSpace
    alphaBoolBlend {blendInTime=1, blendOption=Linear, blendOutTime=1}
    boneToModify {boneName=dyn_skirt_bl_b_02}
    rotation {pitch=10, yaw=-30}
   }
   n16 = "Transform (Modify) Bone - Bone: dyn_skirt_bl_a_master"!(
    ComponentPose=@n15
    bAlphaBoolEnabled:bool=Get Frontend Skydive() -> (Frontend Skydive:bool)
    Translation="0,-5,0"
    Rotation="10,15,0"
    Scale="1,1,1"
   ) -> (Pose) {
    alphaInputType=Bool
    rotationMode=BMM_Additive
    rotationSpace=BCS_BoneSpace
    translationMode=BMM_Additive
    translationSpace=BCS_BoneSpace
    alphaBoolBlend {blendInTime=1, blendOption=Linear, blendOutTime=1}
    alphaScaleBiasClamp {bInterpResult=true}
    boneToModify {boneName=dyn_skirt_bl_a_master}
    rotation {pitch=10, yaw=15}
    translation {y=-5}
   }
   n17 = "Transform (Modify) Bone - Bone: dyn_skirt_l_master"(
    ComponentPose=Local To Component(LocalPose=Copy Pose From Mesh() {bUseAttachedParent=true})
    Alpha=1
    Translation="0,0,0"
    Rotation=Get Thigh Left Rot()
    Scale="1,1,1"
   ) {
    rotationMode=BMM_Additive
    rotationSpace=BCS_BoneSpace
    translationMode=BMM_Additive
    alphaBoolBlend {blendOption=Linear}
    alphaScaleBiasClamp {bInterpResult=true}
    boneToModify {boneName=dyn_skirt_l_master}
   }
   n18 = "Transform (Modify) Bone - Bone: dyn_skirt_fr_master"(ComponentPose=@n17, Alpha=1, Translation="0,0,0", Rotation=Get Thigh Right Rot(), Scale="1,1,1") {
    rotationMode=BMM_Additive
    rotationSpace=BCS_BoneSpace
    translationMode=BMM_Additive
    alphaBoolBlend {blendOption=Linear}
    alphaScaleBiasClamp {bInterpResult=true}
    boneToModify {boneName=dyn_skirt_fr_master}
   }
   n19 = RigidBody(ComponentPose=@n18, Alpha=0) {
    bOverrideWorldGravity=true
    overridePhysicsAsset=&F_MED_Headband_S_Flap_Physics
    simulationSpace=BaseBoneSpace
    alphaBoolBlend {blendInTime=0.5, blendOption=Linear, blendOutTime=0.5}
    baseBoneRef {boneName=pelvis}
    componentAppliedLinearAccClamp=<2000 2000 2000>
    componentLinearAccScale=<0.3 0.3 0.35>
    componentLinearVelScale=<0.3 0.3 0.35>
    overrideWorldGravity {z=-980}
   }
   n20 = RigidBody(ComponentPose=@n19) {
    bOverrideWorldGravity=true
    overridePhysicsAsset=&/F_MED_Headband_S_FlapBL_Physics
    simulationSpace=BaseBoneSpace
    alphaBoolBlend {blendInTime=2, blendOption=Linear, blendOutTime=1}
    alphaScaleBiasClamp {interpSpeedIncreasing=2}
    baseBoneRef {boneName=pelvis}
    componentAppliedLinearAccClamp=<2000 2000 2000>
    componentLinearAccScale=<0.3 0.3 0.35>
    componentLinearVelScale=<0.3 0.3 0.35>
    overrideWorldGravity {z=-980}
   }
   n21 = RigidBody(ComponentPose=@n20) {
    bOverrideWorldGravity=true
    overridePhysicsAsset=&F_MED_Headband_S_FlapFL_Physics
    simulationSpace=BaseBoneSpace
    alphaBoolBlend {blendInTime=2, blendOption=Linear, blendOutTime=1}
    alphaScaleBiasClamp {interpSpeedIncreasing=2}
    baseBoneRef {boneName=pelvis}
    componentAppliedLinearAccClamp=<2000 2000 2000>
    componentLinearAccScale=<0.3 0.3 0.35>
    componentLinearVelScale=<0.3 0.3 0.35>
    overrideWorldGravity {z=-980}
   }
   n22 = RigidBody(ComponentPose=@n21) {
    bOverrideWorldGravity=true
    overridePhysicsAsset=&F_MED_Headband_S_Zipper_Skirt_Physics
    simulationSpace=BaseBoneSpace
    alphaBoolBlend {blendInTime=2, blendOption=Linear, blendOutTime=1}
    alphaScaleBiasClamp {interpSpeedIncreasing=2}
    baseBoneRef {boneName=pelvis}
    componentAppliedLinearAccClamp=<2000 2000 2000>
    componentLinearAccScale=<0.3 0.3 0.35>
    componentLinearVelScale=<0.3 0.3 0.35>
    overrideWorldGravity {z=-980}
   }
   n23 = RigidBody(ComponentPose=@n22) {
    bOverrideWorldGravity=true
    overridePhysicsAsset=&F_MED_Headband_S_Zipper_Shirt_Physics
    simulationSpace=BaseBoneSpace
    alphaBoolBlend {blendInTime=2, blendOption=Linear, blendOutTime=1}
    alphaScaleBiasClamp {interpSpeedIncreasing=2}
    baseBoneRef {boneName=spine_05}
    componentAppliedLinearAccClamp=<2000 2000 2000>
    componentLinearAccScale=<0.3 0.3 0.35>
    componentLinearVelScale=<0.3 0.3 0.35>
    overrideWorldGravity {z=-980}
   }
   n24 = "Transform (Modify) Bone - Bone: dyn_skirt_fr_a_01"!(ComponentPose=@n23, bAlphaBoolEnabled:bool=True, Translation="0,0,0", Rotation="6.348,0.001,0.009", Scale="1,1,1") -> (Pose) {
    alphaInputType=Bool
    rotationMode=BMM_Additive
    rotationSpace=BCS_BoneSpace
    translationMode=BMM_Additive
    translationSpace=BCS_BoneSpace
    alphaBoolBlend {blendInTime=0.5, blendOption=Linear, blendOutTime=1}
    alphaScaleBiasClamp {bInterpResult=true, interpSpeedIncreasing=60}
    boneToModify {boneName=dyn_skirt_fr_a_01}
    rotation=<rot 6.348 0.001 0.009>
   }
   n25 = "Transform (Modify) Bone - Bone: dyn_skirt_fl_a_master"!(ComponentPose=@n24, bAlphaBoolEnabled:bool=True, Translation="0,-1,1", Rotation="0,0,0", Scale="1,1,1") -> (Pose) {
    alphaInputType=Bool
    rotationMode=BMM_Additive
    rotationSpace=BCS_BoneSpace
    translationMode=BMM_Additive
    translationSpace=BCS_BoneSpace
    alphaBoolBlend {blendInTime=0.5, blendOption=Linear, blendOutTime=1}
    alphaScaleBiasClamp {bInterpResult=true, interpSpeedIncreasing=60}
    boneToModify {boneName=dyn_skirt_fl_a_master}
    translation {y=-1, z=1}
   }
   n26 = Get bIsCrouching()
   n27 = "Transform (Modify) Bone - Bone: dyn_skirt_fl_a_master"(ComponentPose=@n25, bAlphaBoolEnabled=@n26, Translation="0,-4,1", Rotation="0,-10,-20", Scale="1,1,1") {
    alphaInputType=Bool
    rotationMode=BMM_Additive
    rotationSpace=BCS_BoneSpace
    translationMode=BMM_Additive
    translationSpace=BCS_BoneSpace
    alphaBoolBlend {blendInTime=0.5, blendOption=Linear, blendOutTime=0.5}
    alphaScaleBiasClamp {bInterpResult=true, interpSpeedIncreasing=60}
    boneToModify {boneName=dyn_skirt_fl_a_master}
    rotation {roll=-20, yaw=-10}
    translation {y=-4, z=1}
   }
   n28 = "Transform (Modify) Bone - Bone: dyn_skirt_fl_b_master"!(ComponentPose=@n27, bAlphaBoolEnabled:bool=@n26, Translation="-0.000,0.5,-0.000", Rotation="0,-10,0", Scale="1,1,1") -> (Pose) {
    alphaInputType=Bool
    rotationMode=BMM_Additive
    rotationSpace=BCS_BoneSpace
    translationMode=BMM_Additive
    translationSpace=BCS_BoneSpace
    alphaBoolBlend {blendInTime=0.5, blendOption=Linear, blendOutTime=0.5}
    alphaScaleBiasClamp {bInterpResult=true, interpSpeedIncreasing=60}
    boneToModify {boneName=dyn_skirt_fl_b_master}
    rotation {yaw=-10}
    translation {y=0.5}
   }
   n29 = "Transform (Modify) Bone - Bone: dyn_skirt_fl_a_02"!(ComponentPose=@n28, bAlphaBoolEnabled:bool=@n26, Translation="0,0,0", Rotation="0,0,0", Scale="1,1,1") -> (Pose) {
    alphaInputType=Bool
    rotationMode=BMM_Additive
    rotationSpace=BCS_BoneSpace
    translationMode=BMM_Additive
    translationSpace=BCS_BoneSpace
    alphaBoolBlend {blendInTime=1, blendOption=Linear, blendOutTime=0.1}
    alphaScaleBiasClamp {bInterpResult=true, interpSpeedIncreasing=60}
    boneToModify {boneName=dyn_skirt_fl_a_02}
   }
   n30 = "Transform (Modify) Bone - Bone: dyn_skirt_fl_b_02"!(ComponentPose=@n29, bAlphaBoolEnabled:bool=@n26, Translation="0,0,0", Rotation="0,10,0", Scale="1,1,1") -> (Pose) {
    alphaInputType=Bool
    rotationMode=BMM_Additive
    rotationSpace=BCS_BoneSpace
    translationMode=BMM_Additive
    translationSpace=BCS_BoneSpace
    alphaBoolBlend {blendInTime=1, blendOption=Linear, blendOutTime=0.1}
    alphaScaleBiasClamp {bInterpResult=true, interpSpeedIncreasing=60}
    boneToModify {boneName=dyn_skirt_fl_b_02}
    rotation {yaw=10}
   }
   n31 = Get Thigh L Crouch Pitch In Alpha() -> (Thigh L Crouch Pitch In Alpha:double)
   n32 = "Transform (Modify) Bone - Bone: dyn_skirt_bl_a_master"!(ComponentPose=@n30, Alpha:float=@n31, Translation="-0.000,-0.000,7", Rotation="0,0,0", Scale="1,1,1") -> (Pose) {
    rotationMode=BMM_Additive
    rotationSpace=BCS_BoneSpace
    translationMode=BMM_Additive
    translationSpace=BCS_BoneSpace
    alphaBoolBlend {blendInTime=1, blendOption=Linear, blendOutTime=1}
    boneToModify {boneName=dyn_skirt_bl_a_master}
    translation {z=7}
   }
   n33 = "Transform (Modify) Bone - Bone: dyn_skirt_bl_b_master"(ComponentPose=@n32, Alpha=@n31, Translation="-0.000,-0.000,3", Rotation="0,0,0", Scale="1,1,1") {
    rotationMode=BMM_Additive
    rotationSpace=BCS_BoneSpace
    translationMode=BMM_Additive
    translationSpace=BCS_BoneSpace
    alphaBoolBlend {blendInTime=1, blendOption=Linear, blendOutTime=1}
    boneToModify {boneName=dyn_skirt_bl_b_master}
    translation {z=3}
   }
   n34 = "Transform (Modify) Bone - Bone: dyn_skirt_l_master"(ComponentPose=@n33, Alpha=@n31, Translation="0,0,0", Rotation="0,-5,0", Scale="1,1,1") {
    rotationMode=BMM_Additive
    rotationSpace=BCS_BoneSpace
    translationMode=BMM_Additive
    translationSpace=BCS_BoneSpace
    alphaBoolBlend {blendInTime=1, blendOption=Linear, blendOutTime=1}
    boneToModify {boneName=dyn_skirt_l_master}
    rotation {yaw=-5}
   }
   n35 = "Transform (Modify) Bone - Bone: dyn_skirt_br_a_master"(ComponentPose=@n34, Alpha=@n31, Translation="0.000,0.000,-4.488", Rotation="-6.223,0.000,0.000", Scale="1,1,1") {
    rotationMode=BMM_Additive
    rotationSpace=BCS_BoneSpace
    translationMode=BMM_Additive
    translationSpace=BCS_BoneSpace
    alphaBoolBlend {blendInTime=1, blendOption=Linear, blendOutTime=1}
    boneToModify {boneName=dyn_skirt_br_a_master}
    rotation {pitch=-6.223}
    translation {z=-4.488}
   }
   n36 = Get Thigh R Crouch Rot Alpha() -> (Thigh R Crouch Rot Alpha:double)
   n37 = "Transform (Modify) Bone - Bone: dyn_skirt_br_a_master"(ComponentPose=@n35, Alpha=@n36, Translation="0,-5,0", Rotation="0.000,-20,0.000", Scale="1,1,1") {
    rotationMode=BMM_Additive
    rotationSpace=BCS_BoneSpace
    translationMode=BMM_Additive
    translationSpace=BCS_BoneSpace
    alphaBoolBlend {blendInTime=1, blendOption=Linear, blendOutTime=1}
    alphaScaleBiasClamp {bInterpResult=true}
    boneToModify {boneName=dyn_skirt_br_a_master}
    rotation {yaw=-20}
    translation {y=-5}
   }
   n38 = "Transform (Modify) Bone - Bone: dyn_skirt_l_master"(ComponentPose=@n37, Alpha=@n36, Translation="0,0,0", Rotation="9.612,-5,15.221", Scale="1,1,1") {
    rotationMode=BMM_Additive
    rotationSpace=BCS_BoneSpace
    translationMode=BMM_Additive
    translationSpace=BCS_BoneSpace
    alphaBoolBlend {blendInTime=1, blendOption=Linear, blendOutTime=1}
    alphaScaleBiasClamp {bInterpResult=true}
    boneToModify {boneName=dyn_skirt_l_master}
    rotation=<rot 9.612 -5 15.221>
   }
   n39 = "Transform (Modify) Bone - Bone: dyn_skirt_bl_a_master"(ComponentPose=@n38, Alpha=Get Headband Emote Alpha(), Translation="0,5,0", Rotation="0,0,0", Scale="1,1,1") {
    rotationMode=BMM_Additive
    rotationSpace=BCS_BoneSpace
    translationMode=BMM_Additive
    translationSpace=BCS_BoneSpace
    alphaBoolBlend {blendInTime=1, blendOption=Linear, blendOutTime=1}
    alphaScaleBiasClamp {bInterpResult=true}
    boneToModify {boneName=dyn_skirt_bl_a_master}
    translation {y=5}
   }
   n40 = "Constraint - dyn_zipper1_base (PPP)"!(ComponentPose=@n39, Alpha=1, ConstraintWeights_0=0.33, ConstraintWeights_1:float=0.33, ConstraintWeights_2:float=0.34) -> (Pose) {
    alphaBoolBlend {blendOption=Linear}
    boneToModify {boneName=dyn_zipper1_base}
    constraintSetup [{targetBone {boneName=thigh_l}}, {targetBone {boneName=thigh_r}}, {targetBone {boneName=pelvis}}]
    constraintWeights [1,1,1]
   }
   n41 = Blend Poses by bool(BlendPose_0=Component To Local(ComponentPose=@n16), BlendPose_1=Component To Local(ComponentPose=@n40), bActiveValue=Get bIsInFrontEnd()) {blendType=Linear}
   Output Pose(Result=Component To Local(ComponentPose=Local To Component(LocalPose=@n41))) {layerGroup=None}
  ]}
 EventGraph {nodes [
   n1 = Event BlueprintUpdateAnimation()
   > Set FPS(FPS:double=float / float(A=1, B=@n1.DeltaTimeX)) -> (Output_Get:double)
   > Set SkirtDynamics(SkirtDynamics:bool="float >= float"(A=Get FPS() -> (FPS:double), B=28)) -> (Output_Get:bool)
   > Set SkirtSkinned(SkirtSkinned:bool="float <= float"(A=Get FPS(), B=28) -> (ReturnValue=~)) -> (Output_Get:bool)
   > n2 = Sequence!(execute) -> (then_0, then_1, then_2:Exec)
   Branch(execute=@n2.then_1, Condition=Get SkirtSkinned() -> (SkirtSkinned:bool))
   > Set RBAN FlapR Alpha(RBAN FlapR Alpha:double=0) -> (Output_Get:double)
   > Set RBAN FlapL Alpha(RBAN FlapL Alpha:double=0) -> (Output_Get:double)
   n3 = GetOwnerSkeletalMeshComponent()
   n4 = BreakRotator(InRot=BreakTransform(InTransform=GetDeltaTransformFromRefPose(self=@n3, BoneName=thigh_l, BaseName=pelvis)).Rotation)
   Set Thigh L Crouch Pitch Alpha(execute=@n2.then_2, Thigh L Crouch Pitch Alpha:double=MapRangeClamped(Value=@n4.Pitch, InRangeA=-20, InRangeB=45)) -> (Output_Get:double)
   > Set Thigh L Crouch Rot Alpha(Thigh L Crouch Rot Alpha:double=MapRangeClamped(Value=@n4.Yaw, InRangeA=-20, InRangeB=-100)) -> (Output_Get:double)
   > Set Thigh L Crouch Pitch In Alpha(Thigh L Crouch Pitch In Alpha:double=MapRangeClamped(Value=@n4, InRangeA=10, InRangeB=60)) -> (Output_Get:double)
   n5 = GetOwnerSkeletalMeshComponent()
   n6 = BreakRotator(InRot=BreakTransform(InTransform=GetDeltaTransformFromRefPose(self=@n5, BoneName=thigh_l, BaseName=pelvis)).Rotation)
   > Set Thigh Left Rot(Thigh Left Rot:Rotator=MakeRotator(Pitch=@n6.Pitch, Yaw=@n6.Yaw)) -> (Output_Get:Rotator)
   n7 = GetOwnerSkeletalMeshComponent()
   n8 = BreakRotator(InRot=BreakTransform(InTransform=GetDeltaTransformFromRefPose(self=@n7, BoneName=thigh_r, BaseName=pelvis)).Rotation)
   > Set Thigh Right Rot(Thigh Right Rot:Rotator=MakeRotator(Pitch=@n8.Pitch, Yaw=@n8.Yaw)) -> (Output_Get:Rotator)
   n9 = BreakTransform(InTransform=GetDeltaTransformFromRefPose(self=GetOwnerSkeletalMeshComponent(), BoneName=thigh_r, BaseName=pelvis))
   > Set Thigh R Crouch Rot Alpha(Thigh R Crouch Rot Alpha:double=MapRangeClamped(Value=BreakRotator(InRot=@n9.Rotation).Yaw, InRangeA=-25, InRangeB=-65)) -> (Output_Get:double)
   > n10 = Branch(Condition=Get bIsPlayingEmote())
   n11 = BreakTransform(InTransform=GetDeltaTransformFromRefPose(self=GetOwnerSkeletalMeshComponent(), BoneName=pelvis, BaseName=root))
   > n12 = Set Headband Emote Alpha(Headband Emote Alpha:double=MapRangeClamped(Value=BreakRotator(InRot=@n11.Rotation).Yaw, InRangeA=15, InRangeB=70)) -> (Output_Get:double)
   Branch(execute=[@n10.else, @n12.then], Condition=Get bIsInFrontEnd())
   n13 = GetOwnerSkeletalMeshComponent()
   n14 = BreakRotator(InRot=BreakTransform(InTransform=GetDeltaTransformFromRefPose(self=@n13, BoneName=pelvis, BaseName=root)).Rotation)
   > Set Pelvis FrontEnd Pose Alpha(Pelvis FrontEnd Pose Alpha:double=MapRangeClamped(Value=@n14, InRangeA=-43, InRangeB=-80)) -> (Output_Get:double)
   > Set Pelvis FrontEnd Pose Inverse Alpha(Pelvis FrontEnd Pose Inverse Alpha:double=MapRangeClamped(Value=@n14, InRangeA=-80, InRangeB=-43)) -> (Output_Get:double)
   > Set RBAN Disabled(RBAN Disabled:bool=NOT Boolean(A=Get bRigidBodyEnabled())) -> (Output_Get:bool)
   n15 = Get bIsSkydiveFloating()
   n16 = Get bIsSkydiveDiveMode()
   > Set Frontend Skydive(Frontend Skydive:bool=AND Boolean(A=Get bIsInFrontEnd(), B=OR Boolean(A=@n16, B=@n15))) -> (Output_Get:bool)
   Branch(execute=@n2.then_0, Condition=Get SkirtDynamics() -> (SkirtDynamics:bool))
   > Set RBAN FlapR Alpha(RBAN FlapR Alpha=0.5)
   > Set RBAN FlapL Alpha(RBAN FlapL Alpha=0.35)
  ]}
]
variables [
 FPS {type=number}
 Frontend Skydive {type=boolean}
 Headband Emote Alpha {type=number}
 Pelvis FrontEnd Pose Alpha {type=number}
 Pelvis FrontEnd Pose Inverse Alpha {type=number}
 RBAN Disabled {type=boolean}
 RBAN FlapL Alpha {type=number}
 RBAN FlapR Alpha {type=number}
 SkirtDynamics {type=boolean}
 SkirtSkinned {type=boolean}
 Thigh L Crouch Pitch Alpha {type=number}
 Thigh L Crouch Pitch In Alpha {type=number}
 Thigh L Crouch Rot Alpha {type=number}
 Thigh Left Rot {type=Rotator}
 Thigh R Crouch Rot Alpha {type=number}
 Thigh Right Rot {type=Rotator}
]
