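# Physics asset for the face-accessory part of the female medium Headband_S player cosmetic, setting up physics for its dynamic hair bones.
# Seven capsule-body chains (back, back-left/right, two front-left, two front-right) hang from tiny base bodies constrained to the head with free rotation.
# Chain links have tight swing and twist limits; both front "a" chains branch, and the left one's root link is angularly locked.
# Larger head and spine_05 capsules represent the body.
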PhysicsAsset /BRCosmetics/Characters/Player/Female/Medium/Bodies/F_MED_Headband_S/Meshes/Parts/F_MED_Headband_S_FaceAcc_Physics
bodies {
 dyn_hair_b_01 [{shapeType=Capsule, length=3.25, radius=2.25, shapeName=dyn_hair_b_01_capsule, rotation=<rot 90 0 0>}]
 dyn_hair_b_02 [{shapeType=Capsule, length=3.25, radius=2.25, shapeName=dyn_hair_b_02_capsule, rotation=<rot 90 0 0>}]
 dyn_hair_b_03 [{shapeType=Capsule, length=3.25, radius=2.25, shapeName=dyn_hair_b_03_capsule, rotation=<rot 90 0 0>}]
 dyn_hair_b_base [{shapeType=Capsule, length=0.1, radius=0.1, shapeName=dyn_hair_b_base_capsule}]
 dyn_hair_bl_01 [{shapeType=Capsule, length=1.5, radius=3.25, shapeName=dyn_hair_bl_01_capsule, rotation=<rot 90 0 0>}]
 dyn_hair_bl_02 [{shapeType=Capsule, length=1.5, radius=3.25, shapeName=dyn_hair_bl_02_capsule, rotation=<rot 90 0 0>}]
 dyn_hair_bl_03 [{shapeType=Capsule, length=1.5, radius=3.25, shapeName=dyn_hair_bl_03_capsule, rotation=<rot 90 0 0>}]
 dyn_hair_bl_base [{shapeType=Capsule, length=0.1, radius=0.1, shapeName=dyn_hair_bl_base_capsule}]
 dyn_hair_br_01 [{shapeType=Capsule, length=1.5, radius=3.25, shapeName=dyn_hair_br_01_capsule, rotation=<rot 90 0 0>}]
 dyn_hair_br_02 [{shapeType=Capsule, length=1.5, radius=3.25, shapeName=dyn_hair_br_02_capsule, rotation=<rot 90 0 0>}]
 dyn_hair_br_03 [{shapeType=Capsule, length=1.5, radius=3.25, shapeName=dyn_hair_br_03_capsule, rotation=<rot 90 0 0>}]
 dyn_hair_br_base [{shapeType=Capsule, length=0.1, radius=0.1, shapeName=dyn_hair_br_base_capsule}]
 dyn_hair_fl_a_01 [{shapeType=Capsule, length=2.25, radius=1.75, shapeName=dyn_hair_fl_a_01_capsule, rotation=<rot 90 0 0>}]
 dyn_hair_fl_a_02 [{shapeType=Capsule, length=2.5, radius=2.25, shapeName=dyn_hair_fl_a_02_capsule, rotation=<rot 90 0 0>}]
 dyn_hair_fl_a_03 [{shapeType=Capsule, length=2.5, radius=2.25, shapeName=dyn_hair_fl_a_03_capsule, rotation=<rot 90 0 0>}]
 dyn_hair_fl_a_04 [{shapeType=Capsule, length=2.5, radius=2.25, shapeName=dyn_hair_fl_a_04_capsule, rotation=<rot 90 0 0>}]
 dyn_hair_fl_a_05 [{shapeType=Capsule, length=2.25, radius=2.75, shapeName=dyn_hair_fl_a_05_capsule, center=<2.4505 0 0>, rotation=<rot 90 0 0>}]
 dyn_hair_fl_a_06 [{shapeType=Capsule, length=2.25, radius=2.75, shapeName=dyn_hair_fl_a_06_capsule, rotation=<rot 90 0 0>}]
 dyn_hair_fl_a_base [{shapeType=Capsule, length=0.1, radius=0.1, shapeName=dyn_hair_fl_a_base_capsule}]
 dyn_hair_fl_b_01 [{shapeType=Capsule, length=1.75, radius=2.25, shapeName=dyn_hair_fl_b_01_capsule, rotation=<rot 90 0 0>}]
 dyn_hair_fl_b_02 [{shapeType=Capsule, length=1.75, radius=2.25, shapeName=dyn_hair_fl_b_02_capsule, rotation=<rot 90 0 0>}]
 dyn_hair_fl_b_base [{shapeType=Capsule, length=0.1, radius=0.1, shapeName=dyn_hair_fl_b_base_capsule}]
 dyn_hair_fr_a_01 [{shapeType=Capsule, length=1.5, radius=3, shapeName=dyn_hair_fr_a_01_capsule, center=<2.4505 0 0>, rotation=<rot 90 0 0>}]
 dyn_hair_fr_a_02 [{shapeType=Capsule, length=2, radius=2.5, shapeName=dyn_hair_fr_a_02_capsule, rotation=<rot 90 0 0>}]
 dyn_hair_fr_a_03 [{shapeType=Capsule, length=2.5, radius=2.25, shapeName=dyn_hair_fr_a_03_capsule, rotation=<rot 90 0 0>}]
 dyn_hair_fr_a_04 [{shapeType=Capsule, length=2.5, radius=2.25, shapeName=dyn_hair_fr_a_04_capsule, rotation=<rot 90 0 0>}]
 dyn_hair_fr_a_05 [{shapeType=Capsule, length=2.5, radius=2.25, shapeName=dyn_hair_fr_a_05_capsule, rotation=<rot 90 0 0>}]
 dyn_hair_fr_a_base [{shapeType=Capsule, length=0.1, radius=0.1, shapeName=dyn_hair_fr_a_base_capsule}]
 dyn_hair_fr_b_01 [{shapeType=Capsule, length=1.75, radius=2.25, shapeName=dyn_hair_fr_b_01_capsule, rotation=<rot 90 0 0>}]
 dyn_hair_fr_b_02 [{shapeType=Capsule, length=1.75, radius=2.25, shapeName=dyn_hair_fr_b_02_capsule}]
 dyn_hair_fr_b_base [{shapeType=Capsule, length=0.1, radius=0.1, shapeName=dyn_hair_fr_b_base_capsule}]
 head [{shapeType=Capsule, length=6.727, radius=6.2244, shapeName=head_capsule, center=<2.5435 0.7524 0>, rotation=<rot 90 0 -13.013>}]
 spine_05 [{shapeType=Capsule, length=13.505, radius=9.255, shapeName=spine_05_capsule, center=<3.7011 1.3739 0>, rotation=<rot 0 -90 0>}]
}
constraints [
 {bone1Name=dyn_hair_fr_b_02, bone2Name=dyn_hair_fr_b_01, swing1LimitDegrees=15, swing2LimitDegrees=15, twistLimitDegrees=15}
 {bone1Name=dyn_hair_bl_01, bone2Name=dyn_hair_bl_base, swing1LimitDegrees=15, swing2LimitDegrees=15, twistLimitDegrees=15}
 {bone1Name=dyn_hair_bl_02, bone2Name=dyn_hair_bl_01, swing1LimitDegrees=15, swing2LimitDegrees=15, twistLimitDegrees=15}
 {bone1Name=dyn_hair_bl_03, bone2Name=dyn_hair_bl_02, swing1LimitDegrees=15, swing2LimitDegrees=15, twistLimitDegrees=15}
 {bone1Name=dyn_hair_b_01, bone2Name=dyn_hair_b_base, swing1LimitDegrees=10, swing2LimitDegrees=10, twistLimitDegrees=5}
 {bone1Name=dyn_hair_b_02, bone2Name=dyn_hair_b_01, swing1LimitDegrees=10, swing2LimitDegrees=10, twistLimitDegrees=5}
 {bone1Name=dyn_hair_b_03, bone2Name=dyn_hair_b_02, swing1LimitDegrees=10, swing2LimitDegrees=10, twistLimitDegrees=5}
 {bone1Name=dyn_hair_br_01, bone2Name=dyn_hair_br_base, swing1LimitDegrees=15, swing2LimitDegrees=15, twistLimitDegrees=15}
 {bone1Name=dyn_hair_br_02, bone2Name=dyn_hair_br_01, swing1LimitDegrees=15, swing2LimitDegrees=15, twistLimitDegrees=15}
 {bone1Name=dyn_hair_br_03, bone2Name=dyn_hair_br_02, swing1LimitDegrees=15, swing2LimitDegrees=15, twistLimitDegrees=15}
 {bone1Name=dyn_hair_fr_a_01, bone2Name=dyn_hair_fr_a_base, swing1LimitDegrees=15, swing2LimitDegrees=15, twistLimitDegrees=15}
 {bone1Name=dyn_hair_fr_a_02, bone2Name=dyn_hair_fr_a_01, swing1LimitDegrees=15, swing2LimitDegrees=15, twistLimitDegrees=15}
 {bone1Name=dyn_hair_fr_a_03, bone2Name=dyn_hair_fr_a_base, swing1LimitDegrees=15, swing2LimitDegrees=15, twistLimitDegrees=15}
 {bone1Name=dyn_hair_fr_a_04, bone2Name=dyn_hair_fr_a_03, swing1LimitDegrees=8, swing2LimitDegrees=8, twistLimitDegrees=5}
 {bone1Name=dyn_hair_fr_a_05, bone2Name=dyn_hair_fr_a_04, swing1LimitDegrees=8, swing2LimitDegrees=8, twistLimitDegrees=5}
 {
  bone1Name=dyn_hair_fl_a_01
  bone2Name=dyn_hair_fl_a_base
  swing1LimitDegrees=4
  swing1Motion=Locked
  swing2LimitDegrees=4
  swing2Motion=Locked
  twistLimitDegrees=2
  twistMotion=Locked
 }
 {bone1Name=dyn_hair_fl_a_02, bone2Name=dyn_hair_fl_a_01, swing1LimitDegrees=8, swing2LimitDegrees=8, twistLimitDegrees=5}
 {bone1Name=dyn_hair_fl_a_03, bone2Name=dyn_hair_fl_a_02, swing1LimitDegrees=8, swing2LimitDegrees=8, twistLimitDegrees=5}
 {bone1Name=dyn_hair_fl_a_04, bone2Name=dyn_hair_fl_a_03, swing1LimitDegrees=8, swing2LimitDegrees=8, twistLimitDegrees=5}
 {bone1Name=dyn_hair_fl_a_05, bone2Name=dyn_hair_fl_a_01, swing1LimitDegrees=2, swing2LimitDegrees=2, twistLimitDegrees=2}
 {bone1Name=dyn_hair_fl_a_06, bone2Name=dyn_hair_fl_a_05, swing1LimitDegrees=10, swing2LimitDegrees=10, twistLimitDegrees=5}
 {bone1Name=dyn_hair_fl_b_01, bone2Name=dyn_hair_fl_b_base, swing1LimitDegrees=8, swing2LimitDegrees=8, twistLimitDegrees=5}
 {bone1Name=dyn_hair_fl_b_02, bone2Name=dyn_hair_fl_b_01, swing1LimitDegrees=8, swing2LimitDegrees=8, twistLimitDegrees=5}
 {bone1Name=dyn_hair_fr_b_01, bone2Name=dyn_hair_fr_b_base, swing1LimitDegrees=15, swing2LimitDegrees=15, twistLimitDegrees=15}
 {
  bone1Name=dyn_hair_bl_base
  bone2Name=head
  swing1LimitDegrees=45
  swing1Motion=Free
  swing2LimitDegrees=45
  swing2Motion=Free
  twistLimitDegrees=45
  twistMotion=Free
 }
 {
  bone1Name=dyn_hair_b_base
  bone2Name=head
  swing1LimitDegrees=45
  swing1Motion=Free
  swing2LimitDegrees=45
  swing2Motion=Free
  twistLimitDegrees=45
  twistMotion=Free
 }
 {
  bone1Name=dyn_hair_br_base
  bone2Name=head
  swing1LimitDegrees=45
  swing1Motion=Free
  swing2LimitDegrees=45
  swing2Motion=Free
  twistLimitDegrees=45
  twistMotion=Free
 }
 {
  bone1Name=dyn_hair_fr_a_base
  bone2Name=head
  swing1LimitDegrees=45
  swing1Motion=Free
  swing2LimitDegrees=45
  swing2Motion=Free
  twistLimitDegrees=45
  twistMotion=Free
 }
 {
  bone1Name=dyn_hair_fl_a_base
  bone2Name=head
  swing1LimitDegrees=45
  swing1Motion=Free
  swing2LimitDegrees=45
  swing2Motion=Free
  twistLimitDegrees=45
  twistMotion=Free
 }
 {
  bone1Name=dyn_hair_fl_b_base
  bone2Name=head
  swing1LimitDegrees=45
  swing1Motion=Free
  swing2LimitDegrees=45
  swing2Motion=Free
  twistLimitDegrees=45
  twistMotion=Free
 }
 {
  bone1Name=dyn_hair_fr_b_base
  bone2Name=head
  swing1LimitDegrees=45
  swing1Motion=Free
  swing2LimitDegrees=45
  swing2Motion=Free
  twistLimitDegrees=45
  twistMotion=Free
 }
 {
  bone1Name=head
  bone2Name=spine_05
  swing1LimitDegrees=45
  swing1Motion=Free
  swing2LimitDegrees=45
  swing2Motion=Free
  twistLimitDegrees=45
  twistMotion=Free
 }
]
