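# Anim Blueprint for the female medium Headband_S face accessory, animating its dynamic hair bones and correcting them during aiming.
# It copies the attached parent's pose and runs a head-relative RigidBody simulation with scaled, clamped component motion.
# Additive bone rotations before simulation follow pelvis yaw while targeting and crouching, otherwise zero.
# Further rotations after simulation follow head pitch relative to spine_05, both alphas being computed each update.
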AnimBlueprint /BRCosmetics/Characters/Player/Female/Medium/Bodies/F_MED_Headband_S/Meshes/Parts/F_MED_Headband_S_FaceAcc_AnimBP
target_skeleton=&/F_MED_Headband_S_FaceAcc_Skeleton
graphs [
 AnimGraph {nodes [
   n1 = Get aimMoveUpAlpha() -> (aimMoveUpAlpha:double)
   n2 = "Transform (Modify) Bone - Bone: dyn_hair_br_base"(
    ComponentPose=Local To Component(LocalPose=Copy Pose From Mesh() {bUseAttachedParent=true})
    Alpha=@n1
    Translation="0,0,0"
    Rotation="36.047,-35.741,-4.856"
    Scale="1,1,1"
   ) {
    rotationMode=BMM_Additive
    rotationSpace=BCS_BoneSpace
    alphaBoolBlend {blendOption=Linear}
    alphaScaleBiasClamp {bInterpResult=true}
    boneToModify {boneName=dyn_hair_br_base}
    rotation=<rot 36.047 -35.741 -4.856>
   }
   n3 = "Transform (Modify) Bone - Bone: dyn_hair_b_base"(ComponentPose=@n2, Alpha=@n1, Translation="0,0,0", Rotation="0.000,-20,0.000", Scale="1,1,1") {
    rotationMode=BMM_Additive
    rotationSpace=BCS_BoneSpace
    alphaBoolBlend {blendOption=Linear}
    alphaScaleBiasClamp {bInterpResult=true}
    boneToModify {boneName=dyn_hair_b_base}
    rotation {yaw=-20}
   }
   n4 = "Transform (Modify) Bone - Bone: dyn_hair_bl_base"(ComponentPose=@n3, Alpha=@n1, Translation="0,0,0", Rotation="0.000,-20,0.000", Scale="1,1,1") {
    rotationMode=BMM_Additive
    rotationSpace=BCS_BoneSpace
    alphaBoolBlend {blendOption=Linear}
    alphaScaleBiasClamp {bInterpResult=true}
    boneToModify {boneName=dyn_hair_bl_base}
    rotation {yaw=-20}
   }
   n5 = RigidBody(ComponentPose=@n4) {
    bOverrideWorldGravity=true
    overridePhysicsAsset=&/F_MED_Headband_S_FaceAcc_Physics
    simulationSpace=BaseBoneSpace
    alphaBoolBlend {blendOption=Linear}
    baseBoneRef {boneName=head}
    componentAppliedLinearAccClamp=<2000 2000 2000>
    componentLinearAccScale=<0.3 0.3 0.35>
    componentLinearVelScale=<0.3 0.3 0.35>
    overrideWorldGravity {z=-980}
   }
   n6 = Get aimUpAlpha() -> (aimUpAlpha:double)
   n7 = "Transform (Modify) Bone - Bone: dyn_hair_br_base"(ComponentPose=@n5, Alpha=@n6, Translation="0,0,0", Rotation="0.000,-20,0.000", Scale="1,1,1") {
    rotationMode=BMM_Additive
    rotationSpace=BCS_BoneSpace
    alphaBoolBlend {blendOption=Linear}
    boneToModify {boneName=dyn_hair_br_base}
    rotation {yaw=-20}
   }
   n8 = "Transform (Modify) Bone - Bone: dyn_hair_b_base"(ComponentPose=@n7, Alpha=@n6, Translation="0,0,0", Rotation="0,-10,0", Scale="1,1,1") {
    rotationMode=BMM_Additive
    rotationSpace=BCS_BoneSpace
    alphaBoolBlend {blendOption=Linear}
    boneToModify {boneName=dyn_hair_b_base}
    rotation {yaw=-10}
   }
   n9 = "Transform (Modify) Bone - Bone: dyn_hair_br_02"(ComponentPose=@n8, Alpha=@n6, Translation="0,0,0", Rotation="0,20,0", Scale="1,1,1") {
    rotationMode=BMM_Additive
    rotationSpace=BCS_BoneSpace
    alphaBoolBlend {blendOption=Linear}
    boneToModify {boneName=dyn_hair_br_02}
    rotation {yaw=20}
   }
   n10 = "Transform (Modify) Bone - Bone: dyn_hair_br_03"(ComponentPose=@n9, Alpha=@n6, Translation="0,0,0", Rotation="0,10,0", Scale="1,1,1") {
    rotationMode=BMM_Additive
    rotationSpace=BCS_BoneSpace
    alphaBoolBlend {blendOption=Linear}
    boneToModify {boneName=dyn_hair_br_03}
    rotation {yaw=10}
   }
   Output Pose(Result=Component To Local(ComponentPose=@n10)) {layerGroup=None}
  ]}
 EventGraph {nodes [
   Event BlueprintUpdateAnimation()
   n1 = GetOwnerSkeletalMeshComponent()
   n2 = BreakTransform(InTransform=GetDeltaTransformFromRefPose(self=@n1, BoneName=head, BaseName=spine_05))
   > Set aimUpAlpha(aimUpAlpha:double=MapRangeClamped(Value=BreakRotator(InRot=@n2.Rotation).Pitch, InRangeA=20, InRangeB=45)) -> (Output_Get:double)
   > n3 = Branch(Condition=AND Boolean(A=Get bIsTargeting(), B=Get bIsCrouching()))
   n4 = BreakTransform(InTransform=GetDeltaTransformFromRefPose(self=@n1, BoneName=pelvis, BaseName=root))
   > Set aimMoveUpAlpha(aimMoveUpAlpha:double=MapRangeClamped(Value=BreakRotator(InRot=@n4.Rotation).Yaw, InRangeA=0, InRangeB=25)) -> (Output_Get:double)
   Set aimMoveUpAlpha(execute=@n3.else, aimMoveUpAlpha=0)
  ]}
]
variables [
 aimMoveUpAlpha {type=number}
 aimUpAlpha {type=number}
]
